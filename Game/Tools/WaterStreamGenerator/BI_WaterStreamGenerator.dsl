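; Blueprint interface for the water stream generator tool, judging by its name and folder.
; It declares a single function, DelayExecuteSnap, with no parameters or return values shown.
; The name suggests a deferred snap call, but the file holds no logic, so the behavior is left to implementing Blueprints.
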(fn DelayExecuteSnap)
