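; Placeable water stream generator actor that configures stream particles, mesh, decal and looping sound in its construction script; its event handlers are empty.
; Actor scale drives decal size, mesh material scaling and Niagara ScaleZ (fourth root of Z), while the particles' world scale stays at one.
; Actor Y location seeds a material offset for variation.
; Flags remove decal or sound or hide StaticMesh2, and an editor component is targeted at the actor.
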(event EventBeginPlay)

(event Collision|EventActorBeginOverlap
  (params (OtherActor :type Actor%20Object%20Reference)))

(event EventTick
  (params (DeltaSeconds :type Float%20%28single-precision%29)))

(event EventDelayExecuteSnap)

(fn ConstructionScript
  (bind _returnvalue (Transformation|GetActorScale3D))
  (bind _z (.z _returnvalue))
  (bind _returnvalue_2 (Transformation|GetActorScale3D))
  (bind _bp_waterstreamgeneratoreditorcomponent (Variables|Default|GetBP_WaterStreamGeneratorEditorComponent))
  (bind _staticmesh2 (Variables|Default|GetStaticMesh2))
  (bind _decal (Variables|Default|GetDecal))
  (bind _loopsound (Variables|Default|GetLoopSound))
  (Utilities|FlowControl|Sequence
    (:then_0
      (Utilities|IsValid _bp_waterstreamgeneratoreditorcomponent
        (:Is%20Valid
          (Variables|/Game/Tools/WaterStreamGenerator/BP_WaterStreamGeneratorEditorComponent.BP_WaterStreamGeneratorEditorComponent_C|SetTarget self _bp_waterstreamgeneratoreditorcomponent))))
    (:then_1
      (Utilities|FlowControl|Sequence
        (:then_0
          (Effects|Components|ParticleSystem|SetFloatParameter (Variables|Default|GetNS_Streams) "SpriteSpeedMult" (Variables|Default|GetNSSpeedMult))
          (Effects|Components|ParticleSystem|SetFloatParameter (Variables|Default|GetNS_Streams) "SpriteScaleMult" (Variables|Default|GetNSScaleMult)))
        (:then_1
          (Transformation|SetWorldScale3D (Variables|Default|GetNS_Streams) "1.000000,1.000000,1.000000")
          (Transformation|SetWorldScale3D _decal (Math|Vector|MakeVector 1.0 _z (.y _returnvalue)))
          (bind _returnvalue_3 (Rendering|Components|Decal|CreateDynamicMaterialInstance _decal))
          (Rendering|Material|SetScalarParameterValue _returnvalue_3 "ScaleU" _z :!class /Script/Engine.MaterialInstanceDynamic))))
    (:then_2
      (if (not (Variables|Default|GetSpawnDecal))
        (Components|DestroyComponent (Variables|Default|GetDecal))))
    (:then_3
      (if (Variables|Default|GetEnableSound)
        (Audio|Components|Audio|SetSound _loopsound (Variables|Default|GetSound))
        (Transformation|SetRelativeLocation (Variables|Default|GetLoopSound) (Math|Vector|MakeVector 0.0 0.0 (* (.z (Transformation|GetWorldScale (Variables|Default|GetStaticMesh))) -2e+02)) false false)
        (else
          (Components|DestroyComponent _loopsound))))
    (:then_4
      (Niagara|SetNiagaraVariable(Float) (Variables|Default|GetNS_Streams) "ScaleZ" (Math|Float|Sqrt (Math|Float|Sqrt (.z (Transformation|GetActorScale3D))))))
    (:then_5
      (if (Variables|Default|GetDisableAllStaticMesh)
        (Rendering|SetVisibility (Variables|Default|GetStaticMesh2) false false)))
    (:then_7
      (Rendering|Material|SetScalarParameterValueonMaterials _staticmesh2 "ScaleY" (* (.y _returnvalue_2) 3.0))
      (Rendering|Material|SetScalarParameterValueonMaterials _staticmesh2 "ScaleZ" (* (.z _returnvalue_2) 4.0)))
    (:then_8
      (Rendering|Material|SetScalarParameterValueonMaterials _staticmesh2 "RandomOffset" (Math|Float|%25%28Float%29 (.y (Transformation|GetActorLocation)) 114.514)))))
