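; Blueprint component for the water stream generator tool, named as an editor component.
; A custom Snap event checks whether the result of GetTarget is valid, but no action follows the check in this graph.
; BeginPlay and Tick events are present but empty.
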(event EventBeginPlay)

(event EventTick
  (params (DeltaSeconds :type Float%20%28single-precision%29)))

(event Custom|Snap
  (Utilities|IsValid (Variables|Default|GetTarget)))
