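; Editor-configurable generator of thin vertical water streams, rendered through one Niagara system.
; In-editor construction pads and trims per-stream arrays (color, size, opacity, top flow, bottom splash) to match StartPos, snapping each end point directly beneath its start.
; TransitParameters derives stream height and top direction from those arrays and pushes everything into Niagara array parameters on every construction.
; RemoveTargetArrayItem deletes one stream by index and re-pushes; runtime events are empty.
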(event EventBeginPlay)

(event Collision|EventActorBeginOverlap
  (params (OtherActor :type Actor%20Object%20Reference)))

(event EventTick
  (params (DeltaSeconds :type Float%20%28single-precision%29)))

(fn ArrayInitial
  (bind _startpos (Variables|Default|GetStartPos))
  (bind _returnvalue (Utilities|Array|Length _startpos))
  (bind _endpos (Variables|Default|GetEndPos))
  (bind _endpos_2 (Variables|Default|GetEndPos))
  (for (_array_element _array_index) _startpos
    (if (not (Utilities|Array|IsValidIndex _endpos_2 _array_index))
      (Utilities|Array|Add _endpos_2 (- (.location _array_element) "0.000000,0.000000,300.000000"))
      (Utilities|Array|Add (Variables|Default|GetTopFlow) "true")
      (Utilities|Array|Add (Variables|Default|GetWaterColor) "(R=0.010000,G=0.050000,B=0.036667,A=1.000000)")
      (Utilities|Array|Add (Variables|Default|GetStreamSize) 35.0)
      (Utilities|Array|Add (Variables|Default|GetTopSize) "(X=200.000000,Y=200.000000)")
      (Utilities|Array|Add (Variables|Default|GetHighLight) 1.0)
      (Utilities|Array|Add (Variables|Default|GetOpacity) 1.0)
      (Utilities|Array|Add (Variables|Default|GetBottomSplash))
      (Utilities|Array|Add (Variables|Default|GetBottomSplashSize) 1.0)))
  (Utilities|Array|Resize (Variables|Default|GetEndPos) _returnvalue)
  (Utilities|Array|Resize (Variables|Default|GetTopFlow) _returnvalue)
  (Utilities|Array|Resize (Variables|Default|GetWaterColor) _returnvalue)
  (Utilities|Array|Resize (Variables|Default|GetStreamSize) _returnvalue)
  (Utilities|Array|Resize (Variables|Default|GetTopSize) _returnvalue)
  (Utilities|Array|Resize (Variables|Default|GetOpacity) _returnvalue)
  (Utilities|Array|Resize (Variables|Default|GetHighLight) _returnvalue)
  (Utilities|Array|Resize (Variables|Default|GetBottomSplash) _returnvalue)
  (Utilities|Array|Resize (Variables|Default|GetBottomSplashSize) _returnvalue)
  (for (_array_element_2 _array_index_2) _endpos
    (Utilities|Array|SetArrayElem _endpos _array_index_2 (Math|Vector|MakeVector (.x (.location (Utilities|Array|Get(acopy) (Variables|Default|GetStartPos) _array_index_2))) (.y (.location (Utilities|Array|Get(acopy) (Variables|Default|GetStartPos) _array_index_2))) (.z _array_element_2)))))

(fn RemoveTargetArrayItem
  (bind _removetargetarrayindex (Variables|Default|GetRemoveTargetArrayIndex))
  (Utilities|Array|RemoveIndex (Variables|Default|GetEndPos) _removetargetarrayindex)
  (Utilities|Array|RemoveIndex (Variables|Default|GetTopFlow) _removetargetarrayindex)
  (Utilities|Array|RemoveIndex (Variables|Default|GetHighLight) _removetargetarrayindex)
  (Utilities|Array|RemoveIndex (Variables|Default|GetWaterColor) _removetargetarrayindex)
  (Utilities|Array|RemoveIndex (Variables|Default|GetBottomSplash) _removetargetarrayindex)
  (Utilities|Array|RemoveIndex (Variables|Default|GetStreamSize) _removetargetarrayindex)
  (Utilities|Array|RemoveIndex (Variables|Default|GetBottomSplashSize) _removetargetarrayindex)
  (Utilities|Array|RemoveIndex (Variables|Default|GetTopSize) _removetargetarrayindex)
  (Utilities|Array|RemoveIndex (Variables|Default|GetOpacity) _removetargetarrayindex)
  (Utilities|Array|RemoveIndex (Variables|Default|GetStartPos) _removetargetarrayindex)
  (CallFunction|TransitParameters))

(fn TransitParameters
  (bind _ns_waterstream (Variables|Default|GetNS_WaterStream))
  (Utilities|FlowControl|Sequence
    (:then_0
      (Utilities|Array|Clear (Variables|Default|GetWaterColorV4))
      (for _array_element (Variables|Default|GetWaterColor)
        (Utilities|Array|Add (Variables|Default|GetWaterColorV4) (Utilities|LinearColortoVector4 _array_element))))
    (:then_1
      (Utilities|Array|Clear (Variables|Default|GetStreamSizeV2))
      (for (_array_element_2 _array_index) (Variables|Default|GetEndPos)
        (Utilities|Array|Add (Variables|Default|GetStreamSizeV2) (Math|Vector2D|MakeVector2D (Utilities|Array|Get(acopy) (Variables|Default|GetStreamSize) _array_index) (- (.z (.location (Utilities|Array|Get(acopy) (Variables|Default|GetStartPos) _array_index))) (.z _array_element_2))))))
    (:then_2
      (Utilities|Array|Clear (Variables|Default|GetStartPosV3))
      (for _array_element_3 (Variables|Default|GetStartPos)
        (Utilities|Array|Add (Variables|Default|GetStartPosV3) (.location _array_element_3))))
    (:then_3
      (Utilities|Array|Clear (Variables|Default|GetTopDirArray))
      (for _array_element_4 (Variables|Default|GetStartPos)
        (Utilities|Array|Add (Variables|Default|GetTopDirArray) (Math|Rotator|GetRotationXVector (.rotation _array_element_4)))))
    (:then_4
      (Niagara|NiagaraSetVectorArray _ns_waterstream "StartPosArray" (Variables|Default|GetStartPosV3))
      (Niagara|NiagaraSetBoolArray _ns_waterstream "TopFlowArray" (Variables|Default|GetTopFlow))
      (Niagara|NiagaraSetVector4Array _ns_waterstream "WaterColorArray" (Variables|Default|GetWaterColorV4))
      (Niagara|NiagaraSetVector2DArray _ns_waterstream "StreamSizeArray" (Variables|Default|GetStreamSizeV2))
      (Niagara|NiagaraSetVector2DArray _ns_waterstream "TopSizeArray" (Variables|Default|GetTopSize))
      (Niagara|NiagaraSetFloatArray _ns_waterstream "OpacityScaleArray" (Variables|Default|GetOpacity))
      (Niagara|NiagaraSetFloatArray _ns_waterstream "HighLightArray" (Variables|Default|GetHighLight))
      (Niagara|NiagaraSetBoolArray _ns_waterstream "BottomSplashValidArray" (Variables|Default|GetBottomSplash))
      (Niagara|NiagaraSetFloatArray _ns_waterstream "BottomSplashSizeArray" (Variables|Default|GetBottomSplashSize))
      (Niagara|NiagaraSetVectorArray _ns_waterstream "TopDirArray" (Variables|Default|GetTopDirArray)))))

(fn ConstructionScript
  (if (Editor|InEditor)
    (CallFunction|ArrayInitial))
  (CallFunction|TransitParameters))
