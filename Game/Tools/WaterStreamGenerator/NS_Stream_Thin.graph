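# Niagara system from the WaterStreamGenerator tool that renders many thin water streams in one instance, driven by per-stream User arrays.
# Spawn count follows StartPosArray, and each particle's SpawnGroup indexes position, size, color and other per-stream values.
# Two burst emitters draw persistent Z-aligned stream sprites and optional direction-facing top-flow sprites, feeding dynamic material parameters.
# A rate-based, distance-culled emitter adds animated splashes at stream bottoms, with bool arrays killing unwanted top and splash particles.
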NiagaraSystem /Game/Tools/WaterStreamGenerator/NS_Stream_Thin
dependencies {
 usedDataInterfaces [&NiagaraDataInterfaceCurve, &NiagaraDataInterfaceSpriteRendererInfo]
 usedDynamicInputs []
 usedModules [&SystemState]
}
emitters {
 Empty {
  emitterUpdateScript [
   &EmitterState(
    Inactive Response=::NewEnumerator1
    Life Cycle Mode=::NewEnumerator1
    Loop Behavior=::NewEnumerator1
    Loop Duration Mode=::NewEnumerator1
    UseLoopDelay=false
   )
   &SpawnBurst_Instantaneous(
    Spawn Count=/Niagara/DynamicInputs/Math/GetVectorArrayCount(Vector Array:NiagaraDataInterfaceArrayFloat3=$User.StartPosArray)
    Use Spawn Probability=false
    UseLoopCountLimit=false
   )
  ]
  particleSpawnScript [
   &FN_SetSpawnGroup(Number of Spawn Groups:NiagaraInt32=&GetVectorArrayCount(Vector Array:NiagaraDataInterfaceArrayFloat3=$User.StartPosArray))
   &InitializeParticle(
    Color=&self:GetColorFromV4(Index:NiagaraInt32=$Particles.SpawnGroup, V4:NiagaraDataInterfaceArrayFloat4=$User.WaterColorArray)
    Lifetime=5
    Position=&SelectVectorFromArray(Direct Array Index=$Particles.SpawnGroup, Vector Selection Array=$User.StartPosArray)
    Position Mode=::NewEnumerator1
    Sprite Size=&SelectVector2DFromArray(Direct Array Index=$Particles.SpawnGroup, Vector 2D Selection Array=$User.StreamSizeArray)
    Sprite Size Mode=::NewEnumerator4
    UsePositionOffset=false
    Write Material Random=true
   )
  ]
  particleUpdateScript [
   &ParticleState(Kill Particles When Lifetime Has Elapsed=false, Let Infinitely Lived Particles Die When Emitter Deactivates=true)
   &SpriteFacingAndAlignment(Sprite Alignment=<0 0 1>, WriteAlignment=true, WriteFacing=false)
   &DynamicMaterialParameters(
    Index 0 Param 1=&SelectFloatFromArray(Direct Array Index=$Particles.SpawnGroup, Float Selection Array=$User.HighLightArray)
    Index 0 Param 2=&SelectFloatFromArray(Direct Array Index=$Particles.SpawnGroup, Float Selection Array=$User.OpacityScaleArray)
    Index 0 Param 3=1
    Index 0 Param 4=&RandomRangeFloat(Minimum=0, Maximum=100)
    Index 1 Param 1=&Divide_Float(A=&MakeFloatFromVector2d(Vector 2D=$Particles.SpriteSize, Channel=::NewEnumerator1), B=2000)
    Module.Param0WriteEnabled=true
    Module.Param0WriteEnabled1=true
    Module.Param1WriteEnabled=true
    Module.Param1WriteEnabled1=true
    Module.Param2WriteEnabled=true
    Module.Param2WriteEnabled1=true
    Module.Param3WriteEnabled=true
    Module.Param3WriteEnabled1=true
    Write Parameter Index 1=true
   )
  ]
  rendererClasses [&NiagaraSpriteRendererProperties]
  renderers [{rendererClass=&NiagaraSpriteRendererProperties}]
 }
 Empty001 {
  emitterUpdateScript [
   &EmitterState(
    Inactive Response=::NewEnumerator1
    Life Cycle Mode=::NewEnumerator1
    Loop Behavior=::NewEnumerator1
    Loop Duration Mode=::NewEnumerator1
    UseLoopDelay=false
   )
   &SpawnBurst_Instantaneous(
    Spawn Count=&GetVectorArrayCount(Vector Array:NiagaraDataInterfaceArrayFloat3=$User.StartPosArray)
    Use Spawn Probability=false
    UseLoopCountLimit=false
   )
  ]
  particleSpawnScript [
   &FN_SetSpawnGroup(Number of Spawn Groups:NiagaraInt32=&GetVectorArrayCount(Vector Array:NiagaraDataInterfaceArrayFloat3=$User.StartPosArray))
   &InitializeParticle(
    Color=&self:GetColorFromV4(Index:NiagaraInt32=$Particles.SpawnGroup, V4:NiagaraDataInterfaceArrayFloat4=$User.WaterColorArray)
    Lifetime=5
    Position=&SelectVectorFromArray(Direct Array Index=$Particles.SpawnGroup, Vector Selection Array=$User.StartPosArray)
    Position Mode=::NewEnumerator1
    Sprite Size=&SelectVector2DFromArray(Direct Array Index=$Particles.SpawnGroup, Vector 2D Selection Array=$User.TopSizeArray)
    Sprite Size Mode=::NewEnumerator4
    UsePositionOffset=false
    Write Material Random=true
   )
   &KillParticles(Kill Particles=&InvertBool(Boolean:NiagaraBool=&self:GetBoolFromArray(BoolArray:NiagaraDataInterfaceArrayBool=$User.TopFlowArray, Int:NiagaraInt32=$Particles.SpawnGroup)))
  ]
  particleUpdateScript [
   &ParticleState(Kill Particles When Lifetime Has Elapsed=false, Let Infinitely Lived Particles Die When Emitter Deactivates=true)
   &SpriteFacingAndAlignment(
    Facing Coordinate Space=::World
    Sprite Alignment=<0 0 1>
    Sprite Facing=&SelectVectorFromArray(Direct Array Index=$Particles.SpawnGroup, Vector Selection Array=$User.TopDirArray)
    WriteAlignment=true
    WriteFacing=true
   )
   &DynamicMaterialParameters(
    Index 0 Param 1=&SelectFloatFromArray(Direct Array Index=$Particles.SpawnGroup, Float Selection Array=$User.HighLightArray)
    Index 0 Param 2=&SelectFloatFromArray(Direct Array Index=$Particles.SpawnGroup, Float Selection Array=$User.OpacityScaleArray)
    Index 0 Param 3=1
    Module.Param0WriteEnabled=true
    Module.Param1WriteEnabled=true
    Module.Param2WriteEnabled=true
    Module.Param3WriteEnabled=true
   )
  ]
  rendererClasses [&NiagaraSpriteRendererProperties]
  renderers [{rendererClass=&NiagaraSpriteRendererProperties}]
 }
 Empty002 {
  emitterUpdateScript [
   &EmitterState(
    Enable Distance Culling=true
    Inactive Response=::NewEnumerator1
    Life Cycle Mode=::NewEnumerator1
    Loop Behavior=::NewEnumerator1
    Loop Duration Mode=::NewEnumerator1
    Module.bUseMaxDistance=true
    Module.bUseMinDistance=false
    Scalability Mode=::NewEnumerator1
    UseLoopDelay=false
   )
   &SpawnRate(
    SpawnRate=&Multiply_Float(
     A=25
     B=&MakeFloatFromInt(Int=&GetVectorArrayCount(Vector Array:NiagaraDataInterfaceArrayFloat3=$User.StartPosArray))
    )
    Use Spawn Probability=false
   )
  ]
  particleSpawnScript [
   &FN_SetSpawnGroup(Number of Spawn Groups:NiagaraInt32=&GetVectorArrayCount(Vector Array:NiagaraDataInterfaceArrayFloat3=$User.StartPosArray))
   &InitializeParticle(
    Color=<rgba 0.25 0.25 0.25 1>
    Lifetime=&RandomRangeFloat(Minimum=0.2, Maximum=0.3)
    Sprite Rotation Angle=&RandomRangeFloat(Minimum=-0.05, Maximum=0.05)
    Sprite Rotation Mode=::NewEnumerator1
    Sprite Size Mode=::NewEnumerator1
    Uniform Sprite Size Max=100
    Uniform Sprite Size Min=25
    UsePositionOffset=false
    Write Material Random=true
   )
   &ShapeLocation(
    MaskSpawnGroup=false
    Non Uniform Scale=<1 1 0>
    Offset=&MakeVector(
     X=0
     Y=0
     Z=&Multiply_Float(
      A=&MakeFloatFromVector2d(
       Vector 2D=&SelectVector2DFromArray(Direct Array Index=$Particles.SpawnGroup, Vector 2D Selection Array=$User.StreamSizeArray)
       Channel=::NewEnumerator1
      )
      B=-1
     )
    )
    Offset Mode=::NewEnumerator0
    Override Seed=false
    Shape Origin=&SelectVectorFromArray(Direct Array Index=$Particles.SpawnGroup, Vector Selection Array=$User.StartPosArray)
    Sphere Radius=50
    UseOwnerScale=true
   )
   &KillParticles(Kill Particles=&InvertBool(Boolean:NiagaraBool=&self:GetBoolFromArray(BoolArray:NiagaraDataInterfaceArrayBool=$User.BottomSplashValidArray, Int:NiagaraInt32=$Particles.SpawnGroup)))
  ]
  particleUpdateScript [
   &ParticleState()
   &ScaleSpriteSize(
    Scale Factor=&MakeVector2D(
     X=&Subtract_Float(
      A=&Multiply_Float(
       A=&MakeFloatFromInt(Int=&RandomRangeInt(Minimum=0, Maximum=1))
       B=2
      )
      B=1
     )
     Y=1
    )
    Scale Sprite Size Mode=::NewEnumerator2
   ) {module=ScaleSpriteSize001}
   &ScaleSpriteSize(
    Uniform Curve Scale=&SelectFloatFromArray(Direct Array Index=$Particles.SpawnGroup, Float Selection Array=$User.BottomSplashSizeArray)
    Uniform Curve Sprite Scale=DataInterface() {propertyValues {Curve {keys [
        {arriveTangent=6.5427, interpMode=RCIM_Cubic, leaveTangent=6.5427, tangentMode=RCTM_User, time=0, value=0}
        {arriveTangent=0.073, interpMode=RCIM_Cubic, leaveTangent=0.073, tangentMode=RCTM_User, time=0.4, value=1}
       ]}}}
   )
   &GravityForce(Gravity=<0 0 -471.2732>)
   &SolveForcesAndVelocity(Clamp Velocity=false, Limit Acceleration=false)
   &SubUVAnimation(
    End Frame Range Override=9
    "Sprite Renderer "=DataInterface() {propertyValues {SpriteRenderer=&self:Empty002_3.NiagaraSpriteRendererProperties_1}}
    UseEndFrame=true
    UseStartFrame=true
   )
   &ScaleColor(
    Scale Alpha=&FloatFromCurve(FloatCurve=DataInterface() {propertyValues {Curve {keys [
         {arriveTangent=6.0288, interpMode=RCIM_Cubic, leaveTangent=6.0288, tangentMode=RCTM_User, time=0, value=0}
         {arriveTangent=-1.4588, interpMode=RCIM_Cubic, leaveTangent=-1.4588, tangentMode=RCTM_User, time=0.17, value=1}
         {arriveTangent=-1.7598, interpMode=RCIM_Cubic, leaveTangent=-1.7598, tangentMode=RCTM_User, time=1, value=0}
        ]}}})
    ScaleA=true
    ScaleRGB=true
   )
  ]
  rendererClasses [&NiagaraSpriteRendererProperties]
  renderers [{rendererClass=&NiagaraSpriteRendererProperties}]
 }
}
summary {
 systemName=NS_Stream_Thin
 emitters [
  {emitterName=Empty, rendererClasses [&NiagaraSpriteRendererProperties]}
  {emitterName=Empty001, rendererClasses [&NiagaraSpriteRendererProperties]}
  {emitterName=Empty002, rendererClasses [&NiagaraSpriteRendererProperties]}
 ]
 userVariables [
  User.BottomSplashSizeArray {
   type=NiagaraDataInterfaceArrayFloat
   defaultValue {
    struct=NiagaraExt_VariableValue_DataInterface
    value {
     dataInterface=&self:NiagaraDataInterfaceArrayFloat_0
     dataInterfaceClass=&NiagaraDataInterfaceArrayFloat
    }
   }
  }
  User.BottomSplashValidArray {
   type=NiagaraDataInterfaceArrayBool
   description=""
   defaultValue {
    struct=NiagaraExt_VariableValue_DataInterface
    value {
     dataInterface=&self:NiagaraDataInterfaceArrayBool_1
     dataInterfaceClass=/Script/Niagara.NiagaraDataInterfaceArrayBool
    }
   }
  }
  User.HighLightArray {
   type=NiagaraDataInterfaceArrayFloat
   defaultValue {
    struct=NiagaraExt_VariableValue_DataInterface
    value {
     dataInterface=&self:NiagaraDataInterfaceArrayFloat_2
     dataInterfaceClass=&NiagaraDataInterfaceArrayFloat
    }
   }
  }
  User.OpacityScaleArray {
   type=NiagaraDataInterfaceArrayFloat
   defaultValue {
    struct=NiagaraExt_VariableValue_DataInterface
    value {
     dataInterface=&self:NiagaraDataInterfaceArrayFloat_1
     dataInterfaceClass=&NiagaraDataInterfaceArrayFloat
    }
   }
  }
  User.StartPosArray {
   type=NiagaraDataInterfaceArrayFloat3
   description=""
   defaultValue {
    struct=NiagaraExt_VariableValue_DataInterface
    value {
     dataInterface=&self:NiagaraDataInterfaceArrayFloat3_0
     dataInterfaceClass=&NiagaraDataInterfaceArrayFloat3
    }
   }
  }
  User.StreamSize {type=Vector2f, description="", defaultValue {struct=Vector2f, value=<0 0>}}
  User.StreamSizeArray {
   type=NiagaraDataInterfaceArrayFloat2
   description=""
   defaultValue {
    struct=NiagaraExt_VariableValue_DataInterface
    value {
     dataInterface=&self:NiagaraDataInterfaceArrayFloat2_0
     dataInterfaceClass=&NiagaraDataInterfaceArrayFloat2
    }
   }
  }
  User.TopDirArray {
   type=NiagaraDataInterfaceArrayFloat3
   description=""
   defaultValue {
    struct=NiagaraExt_VariableValue_DataInterface
    value {
     dataInterface=&self:NiagaraDataInterfaceArrayFloat3_1
     dataInterfaceClass=&NiagaraDataInterfaceArrayFloat3
    }
   }
  }
  User.TopFlowArray {
   type=NiagaraDataInterfaceArrayBool
   description=""
   defaultValue {
    struct=NiagaraExt_VariableValue_DataInterface
    value {
     dataInterface=&self:NiagaraDataInterfaceArrayBool_0
     dataInterfaceClass=&NiagaraDataInterfaceArrayBool
    }
   }
  }
  User.TopSizeArray {
   type=NiagaraDataInterfaceArrayFloat2
   description=""
   defaultValue {
    struct=NiagaraExt_VariableValue_DataInterface
    value {
     dataInterface=&self:NiagaraDataInterfaceArrayFloat2_1
     dataInterfaceClass=&NiagaraDataInterfaceArrayFloat2
    }
   }
  }
  User.WaterColorArray {
   type=NiagaraDataInterfaceArrayFloat4
   description=""
   defaultValue {
    struct=NiagaraExt_VariableValue_DataInterface
    value {
     dataInterface=&self:NiagaraDataInterfaceArrayFloat4_0
     dataInterfaceClass=/Script/Niagara.NiagaraDataInterfaceArrayFloat4
    }
   }
  }
 ]
}
